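# Boss variant of the Medallion inventory's empty-slot list, a CommonUserWidget stacking two placeholder images vertically.
# The images are cached in an array on Construct, or on PreConstruct at design time.
# UpdateVisual takes an augment count, hides that many leading placeholders and shows the rest as SelfHitTestInvisible.
# Both start hidden, and the second image's brush draws nothing, so only the first can render an empty slot.
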WidgetBlueprint /MedallionUI/Inventory/WBP_Medallion_Inventory_EmptySlotList_Boss
graphs [
 EventGraph {nodes [
   n1 = Event PreConstruct()
   > n2 = Branch(Condition=@n1.IsDesignTime)
   MakeArray(execute=[@n2.then, @n3.then])
   n3 = Event Construct()
   Event Tick()
  ]}
 MakeArray {nodes [
   MakeArray() {is=title K2Node_FunctionEntry}
   > Set EmptyArray(EmptyArray:Array of Image Object References=Make Array!(
     "[0]":Image Object Reference=Get Empty_0() -> (Empty_0:Image Object Reference)
     "[1]":Image Object Reference=Get Empty_1() -> (Empty_1:Image Object Reference)
    ) -> (Array:Array of Image Object References)) -> (Output_Get:Array of Image Object References)
  ]}
 UpdateVisual {nodes [
   n1 = UpdateVisual!() -> (then, InNumAugments:int) {is=title K2Node_FunctionEntry}
   > Set NumAugments(NumAugments:int=int - int(A=@n1, B=1) {is=title K2Node_CallFunction}) -> (Output_Get:int=0)
   > n2 = For Each Loop(Array:Array of Image Object References=Get EmptyArray() -> (EmptyArray:Array of Image Object References)) -> (Array Element:Image Object Reference)
   n3 = Get NumAugments() -> (NumAugments:int=0)
   > SetVisibility(
    self=@n2
    InVisibility=Select(
     Option 0=SelfHitTestInvisible
     Option 1=Hidden
     Index="integer <= integer"!(A:int=@n2.Array Index, B:int=@n3) -> (ReturnValue:bool) {is=title K2Node_CallFunction}
    )
   )
  ]}
]
variables [
 EmptyArray {type="array<Image>", default [null,null,null,null]}
 NumAugments {type=integer}
]
widgets {
 info {inheritedWidgetCount=0, parentClass=&CommonUserWidget, rootWidgetClass=&VerticalBox, widgetCount=3}
 widgets [
  VerticalBox_0 {widgetClassPath=&VerticalBox, widget=VerticalBox_0}
  Empty_0 {
   widgetClassPath=&Image
   bIsVariable=true
   parent=VerticalBox_0
   widget=Empty_0
   properties {
    visibility=Hidden
    brush {
     bIsDynamicallyLoaded=false
     drawAs=Box
     imageType=FullColor
     mirroring=NoMirror
     resourceName=None
     resourceObject=/Game/Athena/HUD/Inventory/Art/M_UI_Inventory_ResourceItem_EmptySlot
     tiling=NoTile
     imageSize=<66 66>
     margin=<ltrb 0.5 0.5 0.5 0.5>
     outlineSettings {
      bUseBrushTransparency=false
      roundingType=HalfHeightRadius
      width=0
      color {colorUseRule=UseColor_Specified, specifiedColor=<rgba 0 0 0 0>}
      cornerRadii=<0 0 0 0>
     }
     tintColor {colorUseRule=UseColor_Specified, specifiedColor=<rgba 1 1 1 0.75>}
     uVRegion {bIsValid=false, max=<0 0>, min=<0 0>}
    }
   }
   slot {class=VerticalBoxSlot, properties {horizontalAlignment=HAlign_Left, verticalAlignment=VAlign_Center}}
  }
  Empty_1 {
   widgetClassPath=&Image
   bIsVariable=true
   parent=VerticalBox_0
   widget=Empty_1
   properties {
    visibility=Hidden
    brush {
     bIsDynamicallyLoaded=false
     drawAs=NoDrawType
     imageType=FullColor
     mirroring=NoMirror
     resourceName=None
     resourceObject=&M_UI_Inventory_ResourceItem_EmptySlot
     tiling=NoTile
     imageSize=<0 0>
     margin=<ltrb 0.5 0.5 0.5 0.5>
     outlineSettings {
      bUseBrushTransparency=false
      roundingType=HalfHeightRadius
      width=0
      color {colorUseRule=UseColor_Specified, specifiedColor=<rgba 0 0 0 0>}
      cornerRadii=<0 0 0 0>
     }
     tintColor {colorUseRule=UseColor_Specified, specifiedColor=<rgba 1 1 1 0.75>}
     uVRegion {bIsValid=false, max=<0 0>, min=<0 0>}
    }
   }
   slot {class=VerticalBoxSlot, properties {horizontalAlignment=HAlign_Left, verticalAlignment=VAlign_Center}}
  }
 ]
}
