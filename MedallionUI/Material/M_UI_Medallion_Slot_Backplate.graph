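# Translucent UI material that draws a dark backplate with a slanted four-sided outline for a medallion slot.
# Lerped UV edge distances combine through Min, and two SmoothSteps carve out a thin stroke, shaded by a reflected gray gradient over near-black.
# A separate alpha expression adds a faint gradient-weighted stroke to a horizontal falloff.
# The MedallionShineFX static switch blends in a dim gold shine driven by the cosine of a horizontally panning UV.
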Material /MedallionUI/Material/M_UI_Medallion_Slot_Backplate
graph {
 properties {bAutomaticallySetUsageInEditor=false, blendMode=BLEND_Translucent, materialDomain=MD_UI, usedShadingModels=Default Lit}
 nodes [
  n1 = TextureCoordinate()
  n2 = Add(A=Multiply(A=ComponentMask(None=@n1) {r=true}) {constB=5.8}) {constB=-0.032}
  n3 = Add(A=Multiply(A=ComponentMask(None=@n1) {r=true}) {constB=5.8}) {constB=-0.085}
  n4 = ComponentMask(None=@n1) {g=true}
  n5 = OneMinus(None=@n1)
  n6 = Add(A=Multiply(A=ComponentMask(None=@n5) {r=true}) {constB=5.8}) {constB=-0.035}
  n7 = OneMinus(None=@n1)
  n8 = Add(A=Multiply(A=ComponentMask(None=@n7) {r=true}) {constB=5.8}) {constB=-0.09}
  n9 = ComponentMask(None=@n1) {g=true}
  n10 = Add(A=ComponentMask(None=@n1) {g=true}) {constB=-0.08}
  n11 = Add(A=ComponentMask(None=@n1) {g=true}) {constB=-0.025}
  n12 = ComponentMask(None=@n1) {r=true}
  n13 = Add(A=ComponentMask(None=OneMinus(None=@n1)) {g=true}) {constB=-0.04}
  n14 = Add(A=ComponentMask(None=OneMinus(None=@n1)) {g=true}) {constB=-0.07}
  n15 = ComponentMask(None=@n1) {r=true}
  n16 = Min(
   A=Min(A=LinearInterpolate(A=@n2, B=@n3, Alpha=@n4), B=LinearInterpolate(A=@n6, B=@n8, Alpha=@n9))
   B=Min(A=LinearInterpolate(A=@n10, B=@n11, Alpha=@n12), B=LinearInterpolate(A=@n13, B=@n14, Alpha=@n15))
  )
  n17 = Min(
   A=SmoothStep(Value=Subtract(A=@n16) {constB=0.02, desc=Stroke weigth}) {constMax=0.015, constMin=0.035}
   B=SmoothStep(Value=OneMinus(None=@n16)) {constMax=0.965, constMin=0.985}
  )
  n18 = MaterialFunctionCall!(UVs, Offset, RotationCenter, RotationAngle) {materialFunction=/FortUILibrary/MaterialFunctions/Gradients/MF_UI_ReflectedGradient}
  n19 = OneMinus(None=@n18.Result)
  n20 = OneMinus(None=ComponentMask(None=TextureCoordinate() {uTiling=0.3}) {r=true})
  Add(A=Multiply(A=@n17, B=Multiply(A=@n19) {constB=0.1}), B=Saturate(None=Add(A=@n20) {constB=0}))
  n21 = LinearInterpolate(
   A=Constant3Vector() {constant=<rgba 0.0729 0.0729 0.0729 0>}
   B=Constant3Vector() {constant=<rgba 0.1875 0.1875 0.1875 0>}
   Alpha=@n18.Result
  )
  n22 = Add(
   A=Multiply(A=Multiply(A=@n21, B=@n17)) {constB=0.5}
   B=Multiply(A=OneMinus(None=@n17), B=Constant3Vector() {constant=<rgba 0.0052 0.0052 0.0052 1>})
  )
  n23 = Panner(Coordinate=TextureCoordinate()) {speedX=-0.2}
  n24 = Cosine(None=Multiply(A=ComponentMask(None=@n23) {r=true}) {constB=0.4})
  n25 = Saturate(None=Multiply(A=Add(A=@n24) {constB=-0.25}) {constB=0.35})
  StaticSwitchParameter(True=LinearInterpolate(A=Constant3Vector() {constant=<rgba 0.0938 0.0773 0.0386 1>}, B=@n22, Alpha=OneMinus(None=@n25)), False=@n22) {parameterName=MedallionShineFX}
 ]
}
