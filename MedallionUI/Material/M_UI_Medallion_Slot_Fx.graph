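# Additive UI material for an animated, glowing effect on medallion slots.
# A slowly rotating color texture drives a three-color parameter blend, with squared products of both the blend and the texture.
# A separate mask multiplies a faster-rotating grayscale texture by a circle shape and a boosted band.
# That band comes from two opposing smoothsteps on one channel of rotating polar coordinates, combined with Min to form a moving sweep.
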Material /MedallionUI/Material/M_UI_Medallion_Slot_Fx
graph {
 properties {bAutomaticallySetUsageInEditor=false, blendMode=BLEND_Additive, materialDomain=MD_UI, usedShadingModels=Default Lit}
 nodes [
  n1 = TextureSample(UVs=Rotator() {speed=0.15}) {texture=/MedallionUI/Textures/T_UI_Boons_Slot_FxColor}
  n2 = MaterialFunctionCall(
   A=VectorParameter() {parameterName=Color 1, defaultValue=<rgba 0.0075 0.0045 0.0026 1>}.RGB
   B=VectorParameter() {parameterName=Color 2, defaultValue=<rgba 0.4479 0.2328 0.175 1>}.RGB
   C=VectorParameter() {parameterName=Color 3, defaultValue=<rgba 0.8542 0.5224 0.4271 1>}.RGB
   Alpha=@n1.RGB
  ) {materialFunction=&3ColorBlend}
  Multiply(A=@n2.Result, B=@n2.Result)
  Multiply(A=@n1.RGB, B=@n1.RGB)
  n3 = Rotator() {speed=0.5}
  n4 = ComponentMask(None=MaterialFunctionCall!(UV Coord=Rotator() {speed=1}, LenghtScale, RadialScale) {materialFunction=/Game/Global/MaterialFunctions/TechArt/MF_PolarCoordinates}.Result) {g=true}
  n5 = Saturate(None=Min(A=SmoothStep(Value=@n4) {constMax=0.5, constMin=-1, constValue=0.5}, B=SmoothStep(Value=@n4) {constMax=0, constMin=1, constValue=0.5}))
  Multiply(
   A=MaterialFunctionCall(Size=Constant() {r=1.1}) {materialFunction=&MF_UI_Circle}.Fill
   B=Multiply(
    A=TextureSample(UVs=@n3) {samplerType=SAMPLERTYPE_Grayscale, texture=/MedallionUI/Textures/T_UI_Boons_Slot_FxAlpha}.R
    B=Multiply(A=@n5) {constB=6}
   )
  )
 ]
}
