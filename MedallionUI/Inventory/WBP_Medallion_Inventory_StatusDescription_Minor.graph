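# Minor medallion status description row for the Fortnite inventory UI, built on AthenaItemIcon, showing an item icon with its name and description.
# When the represented item changes, a valid item shows the populated overlay, hides the empty image and fills both texts from the item's display name and description.
# An invalid item collapses the populated overlay and shows the empty image instead.
# Medallion materials supply the backplate and icon; the four RandomFloat variables are unused in the graph.
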WidgetBlueprint /MedallionUI/Inventory/WBP_Medallion_Inventory_StatusDescription_Minor
graphs [EventGraph {nodes [
   n1 = Event OnPostItemToRepresentChanged!() -> (OutputDelegate, then, ItemToRepresent:Fort Item Object Reference) {is=title K2Node_Event}
   > n2 = Branch(Condition=IsValid(Object=@n1.ItemToRepresent))
   > SetVisibility(self=Get Overlay_Populated() -> (Overlay_Populated:Overlay Object Reference), InVisibility=SelfHitTestInvisible)
   > SetVisibility(self=Get Image_Empty() -> (Image_Empty:Image Object Reference), InVisibility=Collapsed)
   > Set Text(
    Text=GetDisplayName!(self:Fort Item Object Reference=@n1.ItemToRepresent) -> (ReturnValue:Text)
    self=Get Text_MedallionName() -> (Text_MedallionName:Common Text Object Reference)
   )
   > Set Text(
    Text=GetDescription!(self:Fort Item Object Reference=@n1.ItemToRepresent) -> (ReturnValue:Text)
    self=Get Text_Description() -> (Text_Description:Common Text Object Reference)
   )
   SetVisibility(execute=@n2.else, self=Get Overlay_Populated(), InVisibility=Collapsed)
   > SetVisibility(self=Get Image_Empty(), InVisibility=SelfHitTestInvisible)
  ]}]
variables [
 RandomFloat_Bottom {type=number}
 RandomFloat_Left {type=number}
 RandomFloat_Right {type=number}
 RandomFloat_Top {type=number}
]
widgets {
 info {inheritedWidgetCount=8, parentClass=/Script/FortniteUI.AthenaItemIcon, rootWidgetClass=&SizeBox, widgetCount=19}
 widgets [
  SizeBox_All {
   widgetClassPath=&SizeBox
   widget=SizeBox_All
   properties {bOverride_MinDesiredHeight=true, bOverride_WidthOverride=true, heightOverride=53, minDesiredHeight=80, widthOverride=468}
  }
  Overlay_0 {widgetClassPath=&Overlay, parent=SizeBox_All, widget=Overlay_0, slot {class=SizeBoxSlot}}
  Image_Backplate {
   widgetClassPath=&Image
   parent=Overlay_0
   widget=Image_Backplate
   properties {brush {
     bIsDynamicallyLoaded=false
     drawAs=Image
     imageType=FullColor
     mirroring=NoMirror
     resourceName=None
     resourceObject=/MedallionUI/Material/M_UI_Medallion_Slot_Backplate
     tiling=NoTile
     imageSize=<64 64>
     margin=<ltrb 0.0384 0.099 0.0568 0.099>
     outlineSettings {
      bUseBrushTransparency=false
      roundingType=HalfHeightRadius
      width=0
      color {colorUseRule=UseColor_Specified, specifiedColor=<rgba 0 0 0 0>}
      cornerRadii=<0 0 0 0>
     }
     tintColor {colorUseRule=UseColor_Specified, specifiedColor=<rgba 1 1 1 0.8>}
     uVRegion {bIsValid=false, max=<0 0>, min=<0 0>}
    }}
   slot {class=OverlaySlot, properties {horizontalAlignment=HAlign_Fill, verticalAlignment=VAlign_Fill}}
  }
  Image_Empty {
   widgetClassPath=&Image
   bIsVariable=true
   parent=Overlay_0
   widget=Image_Empty
   properties {
    visibility=Collapsed
    brush {
     bIsDynamicallyLoaded=false
     drawAs=NoDrawType
     imageType=NoImage
     mirroring=NoMirror
     resourceName=None
     resourceObject=/Game/Athena/HUD/Inventory/Art/M_UI_Inventory_ResourceItem_EmptySlot
     tiling=NoTile
     imageSize=<48 48>
     margin=<ltrb 0.5 0.5 0.5 0.5>
     outlineSettings {
      bUseBrushTransparency=false
      roundingType=HalfHeightRadius
      width=0
      color {colorUseRule=UseColor_Specified, specifiedColor=<rgba 0 0 0 0>}
      cornerRadii=<0 0 0 0>
     }
     tintColor {colorUseRule=UseColor_Specified, specifiedColor=<rgba 1 1 1 1>}
     uVRegion {bIsValid=false, max=<0 0>, min=<0 0>}
    }
   }
   slot {class=OverlaySlot, properties {verticalAlignment=VAlign_Center}}
  }
  Overlay_Populated {
   widgetClassPath=&Overlay
   bIsVariable=true
   parent=Overlay_0
   widget=Overlay_Populated
   slot {class=OverlaySlot, properties {horizontalAlignment=HAlign_Fill, verticalAlignment=VAlign_Fill, padding=<ltrb 12 12 12 12>}}
  }
  HorizontalBox_0 {
   widgetClassPath=&HorizontalBox
   parent=Overlay_Populated
   widget=HorizontalBox_0
   slot {class=OverlaySlot, properties {horizontalAlignment=HAlign_Fill, verticalAlignment=VAlign_Fill}}
  }
  Overlay_Item {
   widgetClassPath=&Overlay
   bIsVariable=true
   parent=HorizontalBox_0
   widget=Overlay_Item
   slot {class=HorizontalBoxSlot, properties {horizontalAlignment=HAlign_Center, verticalAlignment=VAlign_Center, padding=<ltrb 2 0 0 0>}}
  }
  Image_Item_Fx {
   widgetClassPath=&Image
   parent=Overlay_Item
   widget=Image_Item_Fx
   properties {brush {
     bIsDynamicallyLoaded=false
     drawAs=NoDrawType
     imageType=FullColor
     mirroring=NoMirror
     resourceName=None
     resourceObject=/MedallionUI/Material/M_UI_Medallion_Slot_Fx
     tiling=NoTile
     imageSize=<60 60>
     margin=<ltrb 0 0 0 0>
     outlineSettings {
      bUseBrushTransparency=false
      roundingType=HalfHeightRadius
      width=0
      color {colorUseRule=UseColor_Specified, specifiedColor=<rgba 0 0 0 0>}
      cornerRadii=<0 0 0 0>
     }
     tintColor {colorUseRule=UseColor_Specified, specifiedColor=<rgba 1 1 1 1>}
     uVRegion {bIsValid=false, max=<0 0>, min=<0 0>}
    }}
   slot {class=OverlaySlot, properties {horizontalAlignment=HAlign_Center, verticalAlignment=VAlign_Center, padding=<ltrb -20 -20 -20 -20>}}
  }
  Image_ItemIcon {
   widgetClassPath=&Image
   bInherited=true
   bIsVariable=true
   parent=Overlay_Item
   widget=Image_ItemIcon
   properties {
    visibility=SelfHitTestInvisible
    brush {
     bIsDynamicallyLoaded=false
     drawAs=Image
     imageType=FullColor
     mirroring=NoMirror
     resourceName=None
     resourceObject=/MedallionUI/Material/M_UI_HUD_Inventory_ResourceIcon_Base
     tiling=NoTile
     imageSize=<49 49>
     margin=<ltrb 0 0 0 0>
     outlineSettings {
      bUseBrushTransparency=false
      roundingType=HalfHeightRadius
      width=0
      color {colorUseRule=UseColor_Specified, specifiedColor=<rgba 0 0 0 0>}
      cornerRadii=<0 0 0 0>
     }
     tintColor {colorUseRule=UseColor_Specified, specifiedColor=<rgba 1 1 1 1>}
     uVRegion {bIsValid=false, max=<0 0>, min=<0 0>}
    }
   }
   slot {class=OverlaySlot, properties {horizontalAlignment=HAlign_Center, verticalAlignment=VAlign_Center, padding=<ltrb -0.5 -0.5 -0.5 -0.5>}}
  }
  VerticalBox_Desc {
   widgetClassPath=&VerticalBox
   parent=HorizontalBox_0
   widget=VerticalBox_Desc
   slot {class=HorizontalBoxSlot, properties {verticalAlignment=VAlign_Center, padding=<ltrb 6 4 0 4>, size {sizeRule=Fill, value=1}}}
  }
  Text_MedallionName {
   widgetClassPath=&CommonTextBlock
   bIsVariable=true
   parent=VerticalBox_Desc
   widget=Text_MedallionName
   properties {
    autoWrapText=true
    bAutoCollapseWithEmptyText=true
    bIsScrollingEnabled=false
    lineHeightPercentage=0.9
    mobileFontSizeMultiplier=1.5
    text="[Medallion Name]"
    textTransformPolicy=ToUpper
    font {
     bForceMonospaced=false
     bMaterialIsStencil=false
     fontMaterial=/Game/Athena/HUD/Inventory/Art/M_UI_Inventory_Resourece_FontColor
     fontObject=&BurbankSmall
     letterSpacing=0
     monospacedWidth=1
     size=13.13
     skewAmount=0.2
     typefaceFontName=Bold
     outlineSettings {
      bApplyOutlineToDropShadows=false
      bMiteredCorners=false
      bSeparateFillAlpha=false
      outlineMaterial=null
      outlineSize=0
      outlineColor=<rgba 0 0 0 1>
     }
    }
    shadowColorAndOpacity=<rgba 0 0 0 0.375>
    shadowOffset=<0 1>
   }
   slot {class=VerticalBoxSlot}
  }
  Text_Description {
   widgetClassPath=&CommonTextBlock
   bIsVariable=true
   parent=VerticalBox_Desc
   widget=Text_Description
   properties {
    autoWrapText=true
    bAutoCollapseWithEmptyText=true
    bIsScrollingEnabled=false
    lineHeightPercentage=0.88
    minDesiredWidth=0.75
    mobileFontSizeMultiplier=1.5
    text="[Item description text that very veryvery very very verrrrrrryyyyyyyyyyy loonngggg although maybe not tooo ]"
    colorAndOpacity {colorUseRule=UseColor_Specified, specifiedColor=<rgba 1 1 1 0.8>}
    font {
     bForceMonospaced=false
     bMaterialIsStencil=false
     fontMaterial=null
     fontObject=&BurbankSmall
     letterSpacing=0
     monospacedWidth=1
     size=12.75
     skewAmount=0.2
     typefaceFontName=Medium
     outlineSettings {
      bApplyOutlineToDropShadows=false
      bMiteredCorners=false
      bSeparateFillAlpha=false
      outlineMaterial=null
      outlineSize=0
      outlineColor=<rgba 0 0 0 1>
     }
    }
    shadowColorAndOpacity=<rgba 0 0 0 0.375>
    shadowOffset=<0 1>
   }
   slot {class=VerticalBoxSlot}
  }
  Text_StackCount {widgetClassPath=&CommonTextBlock, bInherited=true, widget=null}
  Text_SecondaryStackCount {widgetClassPath=&CommonTextBlock, bInherited=true, widget=null}
  Image_AmmoIcon {widgetClassPath=&Image, bInherited=true, widget=null}
  Image_SecondaryAmmoIcon {widgetClassPath=&Image, bInherited=true, widget=null}
  Image_RechargeableIcon {widgetClassPath=&Image, bInherited=true, widget=null}
  InfoPanel_SecondaryItem {widgetClassPath=&Overlay, bInherited=true, widget=null}
  ItemTierWidget {widgetClassPath=/Script/FortniteUI.AthenaItemTierWidget, bInherited=true, widget=null}
 ]
}
