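# Minor medallion status slot widget for the inventory UI, derived from Fortnite's AthenaInventoryItemTileButton.
# A root overlay holds the inherited ItemIcon binding, filled by the minor medallion status description widget stretched to both axes.
# The inherited input action widget is left unbound, and the blueprint adds no variables or graph logic, so behavior comes from the parent class.
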WidgetBlueprint /MedallionUI/Inventory/WBP_Medallion_Inventory_StatusSlot_Minor
graphs [EventGraph {links [], nodes []}]
variables []
widgets {
 info {inheritedWidgetCount=2, parentClass=/Script/FortniteUI.AthenaInventoryItemTileButton, rootWidgetClass=&Overlay, widgetCount=3}
 widgets [
  RootOverlay {widgetClassPath=&Overlay, widget=RootOverlay}
  ItemIcon {
   widgetClassPath=&/WBP_Medallion_Inventory_StatusDescription_Minor
   bInherited=true
   bIsVariable=true
   parent=RootOverlay
   widget=ItemIcon
   slot {class=OverlaySlot, properties {horizontalAlignment=HAlign_Fill, verticalAlignment=VAlign_Fill}}
  }
  InputActionWidget {widgetClassPath=&CommonActionWidget, bInherited=true, widget=null}
 ]
}
