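# Inventory group widget listing minor medallions under a "Boons" header, built on AthenaInventoryGroupMinorMedallion and wrapped in a HUD element wrapper.
# An item tile view shares an overlay with an empty-slot view, which receives the item count on inventory updates.
# Extension widgets replace the default header, copying its alignment, until removed.
# Mobile touch disables overscroll and picks list width by aspect ratio; group display follows a boolean, true by default.
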WidgetBlueprint /MedallionUI/Inventory/WBP_Medallion_Inventory_StatusList_Minor
graphs [
 EventGraph {nodes [
   "On Inventory Updated Event (ItemTileView)"() {is=title K2Node_ComponentBoundEvent}
   > UpdateVisual!(
    execute
    self:WBP Medallion Inventory Empty Slot List Minor Object Reference=Get EmptyView() -> (EmptyView:WBP Medallion Inventory Empty Slot List Minor Object Reference)
    InNumAugments:int=GetNumItems!(self:List View Object Reference=Get ItemTileView() -> (ItemTileView:Fort Item Tile View Object Reference)) -> (ReturnValue:int=0)
   ) -> (then)
   n1 = Event BP_OnExtensionWidgetAdded!() -> (OutputDelegate, then, ExtensionSlot:EUIExtensionSlot Enum=Invalid, ExtensionWidget:User Widget Object Reference) {is=title K2Node_Event}
   > Is Valid(InputObject=@n1.ExtensionWidget)
   > SetVisibility(self=Get DefaultHeader() -> (DefaultHeader:Common Text Object Reference), InVisibility=Collapsed)
   > n2 = AddChildToOverlay!(
    execute
    self:Overlay Object Reference=Get Overlay_Header() -> (Overlay_Header:Overlay Object Reference)
    Content:Widget Object Reference=@n1.ExtensionWidget
   ) -> (then, ReturnValue:Overlay Slot Object Reference)
   n3 = SlotAsOverlaySlot(Widget=Get DefaultHeader())
   > Is Valid(InputObject=@n3)
   > SetHorizontalAlignment(
    self=@n2
    InHorizontalAlignment=Get HorizontalAlignment!(self:Overlay Slot Object Reference=@n3) -> (HorizontalAlignment:EHorizontalAlignment Enum=HAlign_Fill)
   )
   > SetVerticalAlignment(
    self=@n2
    InVerticalAlignment=Get VerticalAlignment!(self:Overlay Slot Object Reference=@n3) -> (VerticalAlignment:EVerticalAlignment Enum=VAlign_Fill)
   )
   n4 = Event BP_OnExtensionWidgetRemoved!() -> (OutputDelegate, then, ExtensionWidget:User Widget Object Reference) {is=title K2Node_Event}
   > RemoveFromParent(self=@n4.ExtensionWidget)
   > SetVisibility(self=Get DefaultHeader(), InVisibility=SelfHitTestInvisible)
   Event PreConstruct()
   > SetMobileStylings()
  ]}
 SetMobileStylings {nodes [
   SetMobileStylings() {is=title K2Node_FunctionEntry}
   n1 = GetOwningPlayer()
   > Branch(Condition=AND Boolean(A=IsMobileGame(), B=IsUsingTouch(PlayerController=@n1)))
   n2 = GetViewportSize!() -> (ReturnValue_X:double, ReturnValue_Y:double)
   n3 = float / float(A=@n2, B=@n2.ReturnValue_Y)
   > SetWidthOverride(
    self=Get SizeBox_Boons() -> (SizeBox_Boons:Size Box Object Reference)
    InWidthOverride=Select(Option 0:double=850, Option 1:double=1000, Index="float > float"(A=@n3, B=1.6)) -> (ReturnValue:double)
   )
   > SetAllowOverScroll!(execute, self:List View Base Object Reference=Get ItemTileView(), NewAllowOverscroll:bool=false) -> (then)
  ]}
 ShouldDisplayGroup {nodes [
   ShouldDisplayGroup() {is=title K2Node_FunctionEntry}
   > Return Node!(execute, ReturnValue:bool=Get bShouldDisplayMinorMedallionsGroupList() -> (bShouldDisplayMinorMedallionsGroupList:bool))
  ]}
]
variables [bShouldDisplayMinorMedallionsGroupList {type=boolean, default=true}]
widgets {
 info {
  inheritedWidgetCount=1
  namedSlotCount=1
  parentClass=/Script/MedallionUI.AthenaInventoryGroupMinorMedallion
  rootWidgetClass=/Game/Athena/HUD/FortHUDElementWrapper.FortHUDElementWrapper_C
  widgetCount=10
 }
 widgets [
  FortHUDElementWrapper {
   widgetClassPath=/Game/Athena/HUD/FortHUDElementWrapper
   bIsVariable=true
   widget=FortHUDElementWrapper
   properties {elementTag {tagName=HUD.Medallion.LeftSideInventoryMinor}}
  }
  Overlay_0 {widgetClassPath=&Overlay, namedSlotHost=FortHUDElementWrapper, widget=Overlay_0}
  Image_Backing {
   widgetClassPath=&Image
   parent=Overlay_0
   widget=Image_Backing
   properties {brush {
     bIsDynamicallyLoaded=false
     drawAs=NoDrawType
     imageType=NoImage
     mirroring=NoMirror
     resourceName=None
     resourceObject=/Game/Athena/HUD/Inventory/Art/M_UI_InventoryPanel_Backing
     tiling=NoTile
     imageSize=<8 8>
     margin=<ltrb 0.5 0.5 0.5 0.5>
     outlineSettings {
      bUseBrushTransparency=false
      roundingType=HalfHeightRadius
      width=0
      color {colorUseRule=UseColor_Specified, specifiedColor=<rgba 0 0 0 0>}
      cornerRadii=<0 0 0 0>
     }
     tintColor {colorUseRule=UseColor_Specified, specifiedColor=<rgba 1 1 1 1>}
     uVRegion {bIsValid=false, max=<0 0>, min=<0 0>}
    }}
   slot {class=OverlaySlot, properties {horizontalAlignment=HAlign_Fill, verticalAlignment=VAlign_Fill}}
  }
  VerticalBox_101 {
   widgetClassPath=&VerticalBox
   parent=Overlay_0
   widget=VerticalBox_101
   slot {class=OverlaySlot, properties {padding=<ltrb 0 20 0 0>}}
  }
  Overlay_Header {widgetClassPath=&Overlay, bIsVariable=true, parent=VerticalBox_101, widget=Overlay_Header, slot {class=VerticalBoxSlot}}
  DefaultHeader {
   widgetClassPath=&CommonTextBlock
   bIsVariable=true
   parent=Overlay_Header
   widget=DefaultHeader
   properties {
    mobileFontSizeMultiplier=1.2
    scrollStyle=&TextScrollStyle-Base_C
    text=Boons
    textTransformPolicy=ToUpper
    font {
     bForceMonospaced=false
     bMaterialIsStencil=false
     fontMaterial=/Game/Athena/HUD/Inventory/Art/M_UI_Inventory_CatergoryHeader_FontColor
     fontObject=&BurbankSmall
     letterSpacing=0
     monospacedWidth=1
     size=13.13
     skewAmount=0.2
     typefaceFontName=Black
     outlineSettings {
      bApplyOutlineToDropShadows=false
      bMiteredCorners=false
      bSeparateFillAlpha=true
      outlineMaterial=&M-FontShadow
      outlineSize=2
      outlineColor=<rgba 0 0 0 0.5>
     }
    }
    shadowOffset=<0 0>
   }
   slot {class=OverlaySlot, properties {horizontalAlignment=HAlign_Fill}}
  }
  SizeBox_Boons {
   widgetClassPath=&SizeBox
   bIsVariable=true
   parent=VerticalBox_101
   widget=SizeBox_Boons
   properties {bOverride_MinDesiredHeight=true, bOverride_WidthOverride=true, minDesiredHeight=90, widthOverride=468}
   slot {class=VerticalBoxSlot}
  }
  Overlay_1 {widgetClassPath=&Overlay, parent=SizeBox_Boons, widget=Overlay_1, slot {class=SizeBoxSlot}}
  EmptyView {
   widgetClassPath=&/WBP_Medallion_Inventory_EmptySlotList_Minor
   bIsVariable=true
   parent=Overlay_1
   widget=EmptyView
   slot {class=OverlaySlot}
  }
  ItemTileView {
   widgetClassPath=/Script/FortniteUI.FortItemTileView
   bInherited=true
   bIsVariable=true
   parent=Overlay_1
   widget=ItemTileView
   properties {
    bIsFocusable=false
    childItemWidgetCardSizeOverride=S
    entryHeight=0
    entryWidgetClass=&/WBP_Medallion_Inventory_StatusSlot_Minor.WBP_Medallion_Inventory_StatusSlot_Minor_C
    entryWidth=468
    navigation=&self:WidgetTree.ItemTileView.WidgetNavigation_0
    numDesignerPreviewEntries=4
    selectionMode=None
    tileAlignment=LeftAligned
    verticalEntrySpacing=6
   }
   slot {class=OverlaySlot, properties {horizontalAlignment=HAlign_Fill}}
  }
 ]
}
