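# Inventory group widget, based on AthenaInventoryGroupMedallion, that lists minor medallion status slots under a "Force Ability" header.
# An inherited FortItemTileView shows non-selectable StatusSlot_Minor entries, overlaid with an EmptyView empty-slot list.
# When the tile view's inventory updates, its item count is passed to EmptyView's UpdateVisual.
# ShouldDisplayGroup returns a boolean variable that defaults to true.
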WidgetBlueprint /MedallionUI/Inventory/WBP_Medallion_Inventory_StatusList_PlotTwist
graphs [
 EventGraph {nodes [
   "On Inventory Updated Event (ItemTileView)"() {is=title K2Node_ComponentBoundEvent}
   > UpdateVisual!(
    execute
    self:WBP Medallion Inventory Empty Slot List Minor Object Reference=Get EmptyView() -> (EmptyView:WBP Medallion Inventory Empty Slot List Minor Object Reference)
    InNumAugments:int=GetNumItems!(self:List View Object Reference=Get ItemTileView() -> (ItemTileView:Fort Item Tile View Object Reference)) -> (ReturnValue:int=0)
   ) -> (then)
  ]}
 ShouldDisplayGroup {nodes [
   ShouldDisplayGroup() {is=title K2Node_FunctionEntry}
   > Return Node!(execute, ReturnValue:bool=Get bShouldDisplayMinorMedallionsGroupList() -> (bShouldDisplayMinorMedallionsGroupList:bool))
  ]}
]
variables [bShouldDisplayMinorMedallionsGroupList {type=boolean, default=true}]
widgets {
 info {inheritedWidgetCount=1, parentClass=/Script/MedallionUI.AthenaInventoryGroupMedallion, rootWidgetClass=&Overlay, widgetCount=8}
 widgets [
  Overlay_0 {widgetClassPath=&Overlay, widget=Overlay_0}
  Image_Backing {
   widgetClassPath=&Image
   parent=Overlay_0
   widget=Image_Backing
   properties {brush {
     bIsDynamicallyLoaded=false
     drawAs=NoDrawType
     imageType=NoImage
     mirroring=NoMirror
     resourceName=None
     resourceObject=/Game/Athena/HUD/Inventory/Art/M_UI_InventoryPanel_Backing
     tiling=NoTile
     imageSize=<8 8>
     margin=<ltrb 0.5 0.5 0.5 0.5>
     outlineSettings {
      bUseBrushTransparency=false
      roundingType=HalfHeightRadius
      width=0
      color {colorUseRule=UseColor_Specified, specifiedColor=<rgba 0 0 0 0>}
      cornerRadii=<0 0 0 0>
     }
     tintColor {colorUseRule=UseColor_Specified, specifiedColor=<rgba 1 1 1 1>}
     uVRegion {bIsValid=false, max=<0 0>, min=<0 0>}
    }}
   slot {class=OverlaySlot, properties {horizontalAlignment=HAlign_Fill, verticalAlignment=VAlign_Fill}}
  }
  VerticalBox_101 {
   widgetClassPath=&VerticalBox
   parent=Overlay_0
   widget=VerticalBox_101
   slot {class=OverlaySlot, properties {padding=<ltrb 0 20 0 0>}}
  }
  CommonTextBlock_1 {
   widgetClassPath=&CommonTextBlock
   parent=VerticalBox_101
   widget=CommonTextBlock_1
   properties {
    mobileFontSizeMultiplier=1.2
    scrollStyle=&TextScrollStyle-Base_C
    text=Force Ability
    textTransformPolicy=ToUpper
    font {
     bForceMonospaced=false
     bMaterialIsStencil=false
     fontMaterial=/Game/Athena/HUD/Inventory/Art/M_UI_Inventory_CatergoryHeader_FontColor
     fontObject=&BurbankSmall
     letterSpacing=0
     monospacedWidth=1
     size=13.13
     skewAmount=0.2
     typefaceFontName=Black
     outlineSettings {
      bApplyOutlineToDropShadows=false
      bMiteredCorners=false
      bSeparateFillAlpha=true
      outlineMaterial=&M-FontShadow
      outlineSize=2
      outlineColor=<rgba 0 0 0 0.5>
     }
    }
    shadowOffset=<0 0>
   }
   slot {class=VerticalBoxSlot, properties {verticalAlignment=VAlign_Top, padding=<ltrb -3 0 0 8>}}
  }
  SizeBox_0 {
   widgetClassPath=&SizeBox
   parent=VerticalBox_101
   widget=SizeBox_0
   properties {bOverride_MinDesiredHeight=true, minDesiredHeight=90, widthOverride=468}
   slot {class=VerticalBoxSlot}
  }
  Overlay_1 {widgetClassPath=&Overlay, parent=SizeBox_0, widget=Overlay_1, slot {class=SizeBoxSlot}}
  EmptyView {
   widgetClassPath=&/WBP_Medallion_Inventory_EmptySlotList_Minor
   bIsVariable=true
   parent=Overlay_1
   widget=EmptyView
   slot {class=OverlaySlot}
  }
  ItemTileView {
   widgetClassPath=/Script/FortniteUI.FortItemTileView
   bInherited=true
   bIsVariable=true
   parent=Overlay_1
   widget=ItemTileView
   properties {
    bIsFocusable=false
    childItemWidgetCardSizeOverride=S
    entryHeight=0
    entryWidgetClass=&/WBP_Medallion_Inventory_StatusSlot_Minor.WBP_Medallion_Inventory_StatusSlot_Minor_C
    entryWidth=468
    navigation=&self:WidgetTree.ItemTileView.WidgetNavigation_0
    numDesignerPreviewEntries=4
    selectionMode=None
    tileAlignment=LeftAligned
    verticalEntrySpacing=6
   }
   slot {class=OverlaySlot, properties {horizontalAlignment=HAlign_Fill}}
  }
 ]
}
